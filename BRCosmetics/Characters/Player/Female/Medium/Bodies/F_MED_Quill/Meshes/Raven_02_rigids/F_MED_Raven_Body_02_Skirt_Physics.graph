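# Physics asset for the leather skirt of the F_MED_Quill Raven_02 female player body, defining bodies for its secondary motion.
# Left and right dyn_leather chains each link five box bodies in series from a small base box.
# Swing limits tighten down each chain and twist stays nearly locked, so the panels bend with little rolling.
# Unconstrained thigh and calf capsules give the legs collision shapes alongside the skirt.
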PhysicsAsset /BRCosmetics/Characters/Player/Female/Medium/Bodies/F_MED_Quill/Meshes/Raven_02_rigids/F_MED_Raven_Body_02_Skirt_Physics
bodies {
 calf_l [
  {shapeType=Capsule, length=29.5, radius=6, center=<-23.5219 -0.0002 0.3372>, rotation=<rot 90 0 0>}
  {shapeType=Capsule, length=8, radius=6.75, center=<-17.5708 2.913 -0.1854>, rotation=<rot 90 0 0>}
 ]
 calf_r [
  {shapeType=Capsule, length=29.5, radius=6, center=<22.4015 -0.0002 0.3372>, rotation=<rot 90 0 0>}
  {shapeType=Capsule, length=8, radius=6.75, center=<16.1945 2.9132 0.7302>, rotation=<rot 90 0 0>}
 ]
 dyn_leather_l_01 [{shapeType=Box, extentX=12.8704, extentY=16.4723, extentZ=3.1547, center=<6.32 0.0003 -0.0001>}]
 dyn_leather_l_02 [{shapeType=Box, extentX=11.6834, extentY=16.9284, extentZ=3.6547, center=<5.4274 0 0.0001>}]
 dyn_leather_l_03 [{shapeType=Box, extentX=12.7581, extentY=15.4937, extentZ=4.1349, center=<6.2794 0 -0.5311>, rotation=<rot -4.2407 0 0>}]
 dyn_leather_l_04 [{shapeType=Box, extentX=10.2075, extentY=14.2802, extentZ=2.7606, center=<4.3723 0 0.0001>}]
 dyn_leather_l_05 [{shapeType=Box, extentX=10.5011, extentY=12.4917, extentZ=3, center=<4.519 0 -0.0002>}]
 dyn_leather_l_base [{shapeType=Box, extentX=2, extentY=2, extentZ=2}]
 dyn_leather_r_01 [{shapeType=Box, extentX=12.8704, extentY=16.4723, extentZ=3.1547, center=<6.32 0.0003 -0.0001>}]
 dyn_leather_r_02 [{shapeType=Box, extentX=12.8704, extentY=16.4723, extentZ=3.1547, center=<6.32 0.0003 -0.0001>}]
 dyn_leather_r_03 [{shapeType=Box, extentX=11.6834, extentY=16.9284, extentZ=3.6547, center=<5.4274 0 0.0001>}]
 dyn_leather_r_04 [{shapeType=Box, extentX=10.2075, extentY=14.2802, extentZ=2.7606, center=<4.3723 0 0.0001>}]
 dyn_leather_r_05 [{shapeType=Box, extentX=10.5011, extentY=12.4917, extentZ=3, center=<4.519 0 -0.0002>}]
 dyn_leather_r_base [{shapeType=Box, extentX=2, extentY=2, extentZ=2}]
 thigh_l [{shapeType=Capsule, length=25.75, radius=8.25, center=<-33.3868 0.0011 1.2152>, rotation=<rot 86.1678 -179.9999 -179.9999>}]
 thigh_r [{shapeType=Capsule, length=25.75, radius=8.25, center=<32.552 0.0001 -0.3242>, rotation=<rot 86.1678 -179.9999 -179.9999>}]
}
constraints [
 {bone1Name=dyn_leather_r_01, bone2Name=dyn_leather_r_base, swing1LimitDegrees=40, swing2LimitDegrees=40, twistLimitDegrees=3}
 {bone1Name=dyn_leather_r_02, bone2Name=dyn_leather_r_01, swing1LimitDegrees=30, swing2LimitDegrees=30, twistLimitDegrees=3}
 {bone1Name=dyn_leather_r_03, bone2Name=dyn_leather_r_02, swing1LimitDegrees=25, swing2LimitDegrees=25, twistLimitDegrees=3}
 {bone1Name=dyn_leather_r_04, bone2Name=dyn_leather_r_03, swing1LimitDegrees=20, swing2LimitDegrees=20, twistLimitDegrees=3}
 {bone1Name=dyn_leather_r_05, bone2Name=dyn_leather_r_04, swing1LimitDegrees=20, swing2LimitDegrees=20, twistLimitDegrees=3}
 {bone1Name=dyn_leather_l_05, bone2Name=dyn_leather_l_04, swing1LimitDegrees=20, swing2LimitDegrees=20, twistLimitDegrees=3}
 {bone1Name=dyn_leather_l_04, bone2Name=dyn_leather_l_03, swing1LimitDegrees=20, swing2LimitDegrees=20, twistLimitDegrees=3}
 {bone1Name=dyn_leather_l_03, bone2Name=dyn_leather_l_02, swing1LimitDegrees=25, swing2LimitDegrees=25, twistLimitDegrees=3}
 {bone1Name=dyn_leather_l_02, bone2Name=dyn_leather_l_01, swing1LimitDegrees=30, swing2LimitDegrees=30, twistLimitDegrees=3}
 {bone1Name=dyn_leather_l_01, bone2Name=dyn_leather_l_base, swing1LimitDegrees=40, swing2LimitDegrees=40, twistLimitDegrees=3}
]
