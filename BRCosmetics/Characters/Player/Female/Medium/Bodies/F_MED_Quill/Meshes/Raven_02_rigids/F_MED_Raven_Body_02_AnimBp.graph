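# Post-copy secondary-motion AnimBlueprint for the female Raven body cosmetic, animating feathers, shoulder pads, pouch and collar.
# It copies the attached parent mesh's pose, then chains AnimDynamics, bone-delta copies, weighted constraints and a skirt RigidBody node.
# Skydiving feeds Perlin-noise force from pawn height into several dynamics nodes, halved under parachute, and stiffens cloth anim drive with a one-shot reset.
# Crouching, except crouch-sprinting, applies an interpolated offset to the right leather bone.
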AnimBlueprint /BRCosmetics/Characters/Player/Female/Medium/Bodies/F_MED_Quill/Meshes/Raven_02_rigids/F_MED_Raven_Body_02_AnimBp
target_skeleton=&/F_MED_Raven_Body_02_Skeleton
graphs [
 AnimGraph {nodes [
   n1 = Get falling_anim_dyn_up_force_vec3() -> (falling_anim_dyn_up_force_vec3:Vector)
   n2 = Anim Dynamics - Bone: dyn_feathers_arm_l(ComponentPose=Local To Component(LocalPose=Copy Pose From Mesh() {bUseAttachedParent=true}), Alpha=0.125, ExternalForce=@n1) {
    angularDampingOverride=0.95
    angularSpringConstant=325
    bAngularSpring=true
    bGravityOverrideInSimSpace=true
    bLinearSpring=true
    bOverrideAngularDamping=true
    bOverrideLinearDamping=true
    lODThreshold=1
    linearDampingOverride=0.95
    linearSpringConstant=325
    simulationSpace=Actor
    alphaBoolBlend {blendOption=Linear}
    boundBone {boneName=dyn_feathers_arm_l}
    physicsBodyDefinitions [{
      sphereCollisionRadius=0
      boundBone {boneName=dyn_feathers_arm_l}
      boxExtents=<30 30 30>
      constraintSetup {
       angularTargetAxis=AxisY
       angularLimitsMax=<0 15 15>
       angularLimitsMin=<0 -15 -15>
       angularTarget=<0 0 0>
       linearAxesMax=<10 10 10>
       linearAxesMin=<0 -10 -10>
      }
      localJointOffset=<0 0 0>
     }]
   }
   n3 = Anim Dynamics - Bone: dyn_feathers_arm_r(ComponentPose=@n2, Alpha=0.125, ExternalForce=@n1) {
    angularDampingOverride=0.95
    angularSpringConstant=325
    bAngularSpring=true
    bGravityOverrideInSimSpace=true
    bLinearSpring=true
    bOverrideAngularDamping=true
    bOverrideLinearDamping=true
    lODThreshold=1
    linearDampingOverride=0.95
    linearSpringConstant=325
    simulationSpace=Actor
    alphaBoolBlend {blendOption=Linear}
    boundBone {boneName=dyn_feathers_arm_r}
    physicsBodyDefinitions [{
      sphereCollisionRadius=0
      boundBone {boneName=dyn_feathers_arm_r}
      boxExtents=<30 30 30>
      constraintSetup {
       angularTargetAxis=AxisY
       angularLimitsMax=<0 15 15>
       angularLimitsMin=<0 -15 -15>
       angularTarget=<0 0 0>
       linearAxesMax=<10 10 10>
       linearAxesMin=<0 -10 -10>
      }
      localJointOffset=<0 0 0>
     }]
   }
   n4 = Anim Dynamics - Bone: dyn_shoulderpad_l(ComponentPose=@n3, Alpha=1, ExternalForce=@n1) {
    bGravityOverrideInSimSpace=true
    lODThreshold=1
    numSolverIterationsPostUpdate=4
    numSolverIterationsPreUpdate=16
    alphaBoolBlend {blendOption=Linear}
    boundBone {boneName=dyn_shoulderpad_l}
    physicsBodyDefinitions [{
      sphereCollisionRadius=0
      boundBone {boneName=dyn_shoulderpad_l}
      boxExtents=<10 20 10>
      constraintSetup {angularLimitsMax=<15 15 0>, angularLimitsMin=<-15 -15 0>, angularTarget=<0 0 0>, linearAxesMin=<-5 0 0>}
      localJointOffset=<10 0 -5>
     }]
   }
   n5 = Anim Dynamics - Bone: dyn_shoulderpad_r(ComponentPose=@n4, Alpha=1, ExternalForce=@n1) {
    bGravityOverrideInSimSpace=true
    lODThreshold=1
    numSolverIterationsPostUpdate=4
    numSolverIterationsPreUpdate=16
    alphaBoolBlend {blendOption=Linear}
    boundBone {boneName=dyn_shoulderpad_r}
    physicsBodyDefinitions [{
      sphereCollisionRadius=0
      boundBone {boneName=dyn_shoulderpad_r}
      boxExtents=<10 20 10>
      constraintSetup {angularLimitsMax=<15 15 0>, angularLimitsMin=<-15 -15 0>, angularTarget=<0 0 0>, linearAxesMin=<-5 0 0>}
      localJointOffset=<-10 0 3.5>
     }]
   }
   n6 = Anim Dynamics - Bone: dyn_feathers_shoulder_r_01(ComponentPose=@n5, Alpha=1, ExternalForce=@n1) {
    alpha=0.9
    angularDampingOverride=0.96
    angularSpringConstant=150
    bAngularSpring=true
    bGravityOverrideInSimSpace=true
    bOverrideAngularDamping=true
    gravityScale=0.6
    linearDampingOverride=0.97
    numSolverIterationsPostUpdate=4
    numSolverIterationsPreUpdate=16
    alphaBoolBlend {blendOption=Linear}
    boundBone {boneName=dyn_feathers_shoulder_r_01}
    physicsBodyDefinitions [{
      sphereCollisionRadius=0
      boundBone {boneName=dyn_feathers_shoulder_r_01}
      boxExtents=<15 5 5>
      constraintSetup {angularTargetAxis=AxisZ, twistAxis=AxisZ, angularLimitsMax=<0 25 0>, angularLimitsMin=<0 -25 0>, angularTarget=<0 0 1>}
      localJointOffset=<0 0 0>
     }]
   }
   n7 = Copy Bone Delta - Source Bone: dyn_feathers_shoulder_r_01 - Target Bone: dyn_feathers_shoulder_r_02!(ComponentPose:Component Space Pose Link Structure=@n6, Alpha:float=1) -> (Pose:Component Space Pose Link Structure) {
    is=title AnimGraphNode_CopyBoneDelta
    bCopyRotation=true
    bCopyScale=true
    bCopyTranslation=true
    alphaBoolBlend {blendOption=Linear}
    sourceBone {boneName=dyn_feathers_shoulder_r_01}
    targetBone {boneName=dyn_feathers_shoulder_r_02}
   }
   n8 = Copy Bone Delta - Source Bone: dyn_feathers_shoulder_r_01 - Target Bone: dyn_feathers_shoulder_r_03!(ComponentPose:Component Space Pose Link Structure=@n7, Alpha:float=1) -> (Pose:Component Space Pose Link Structure) {
    is=title AnimGraphNode_CopyBoneDelta
    bCopyRotation=true
    bCopyScale=true
    bCopyTranslation=true
    alphaBoolBlend {blendOption=Linear}
    sourceBone {boneName=dyn_feathers_shoulder_r_01}
    targetBone {boneName=dyn_feathers_shoulder_r_03}
   }
   n9 = Anim Dynamics - Bone: dyn_feathers_shoulder_l_01(ComponentPose=@n8, Alpha=1, ExternalForce=@n1) {
    alpha=0.9
    angularDampingOverride=0.96
    angularSpringConstant=125
    bAngularSpring=true
    bGravityOverrideInSimSpace=true
    gravityScale=0.5
    linearDampingOverride=0.97
    numSolverIterationsPostUpdate=4
    numSolverIterationsPreUpdate=16
    alphaBoolBlend {blendOption=Linear}
    boundBone {boneName=dyn_feathers_shoulder_l_01}
    physicsBodyDefinitions [{
      sphereCollisionRadius=0
      boundBone {boneName=dyn_feathers_shoulder_l_01}
      boxExtents=<15 5 5>
      constraintSetup {angularTargetAxis=AxisZ, twistAxis=AxisZ, angularLimitsMax=<0 15 0>, angularLimitsMin=<0 -15 0>, angularTarget=<0 0 1>}
      localJointOffset=<0 0 0>
     }]
   }
   n10 = Copy Bone Delta - Source Bone: dyn_feathers_shoulder_l_01 - Target Bone: dyn_feathers_shoulder_l_03!(ComponentPose:Component Space Pose Link Structure=@n9, Alpha:float=1) -> (Pose:Component Space Pose Link Structure) {
    is=title AnimGraphNode_CopyBoneDelta
    bCopyRotation=true
    bCopyScale=true
    bCopyTranslation=true
    alphaBoolBlend {blendOption=Linear}
    sourceBone {boneName=dyn_feathers_shoulder_l_01}
    targetBone {boneName=dyn_feathers_shoulder_l_03}
   }
   n11 = Copy Bone Delta - Source Bone: dyn_feathers_shoulder_l_01 - Target Bone: dyn_feathers_shoulder_l_02!(ComponentPose:Component Space Pose Link Structure=@n10, Alpha:float=1) -> (Pose:Component Space Pose Link Structure) {
    is=title AnimGraphNode_CopyBoneDelta
    bCopyRotation=true
    bCopyScale=true
    bCopyTranslation=true
    alphaBoolBlend {blendOption=Linear}
    sourceBone {boneName=dyn_feathers_shoulder_l_01}
    targetBone {boneName=dyn_feathers_shoulder_l_02}
   }
   n12 = Anim Dynamics - Bone: dyn_box(ComponentPose=@n11, Alpha=1, ExternalForce="0,0,0") {
    bGravityOverrideInSimSpace=true
    gravityScale=0.1
    lODThreshold=1
    alphaBoolBlend {blendOption=Linear}
    alphaScaleBiasClamp {bClampResult=true, clampMax=0.33, clampMin=0.1, scale=0.0033}
    boundBone {boneName=dyn_box}
    physicsBodyDefinitions [{
      sphereCollisionRadius=0
      boundBone {boneName=dyn_box}
      boxExtents=<5 5 5>
      constraintSetup {angularTargetAxis=AxisY, twistAxis=AxisY, angularLimitsMax=<0 0 0>, angularLimitsMin=<0 0 0>, angularTarget=<0 0 0>, linearAxesMax=<0 0 2>}
      localJointOffset=<0 0 5>
     }]
   }
   n13 = Copy Bone Delta - Source Bone: dyn_box - Target Bone: dyn_pouch!(ComponentPose:Component Space Pose Link Structure=@n12, Alpha:float=1) -> (Pose:Component Space Pose Link Structure) {
    is=title AnimGraphNode_CopyBoneDelta
    bCopyRotation=true
    bCopyScale=true
    bCopyTranslation=true
    alphaBoolBlend {blendOption=Linear}
    sourceBone {boneName=dyn_box}
    targetBone {boneName=dyn_pouch}
   }
   n14 = Copy Bone Delta - Source Bone: head - Target Bone: dyn_collar!(ComponentPose:Component Space Pose Link Structure=@n13, Alpha:float=1) -> (Pose:Component Space Pose Link Structure) {
    is=title AnimGraphNode_CopyBoneDelta
    bCopyRotation=true
    bCopyScale=true
    bCopyTranslation=true
    alphaBoolBlend {blendOption=Linear}
    sourceBone {boneName=head}
    targetBone {boneName=dyn_collar}
   }
   n15 = Anim Dynamics - Bone: dyn_collar(ComponentPose=@n14, Alpha=0.126, ExternalForce=@n1) {
    angularSpringConstant=20
    bAngularSpring=true
    bGravityOverrideInSimSpace=true
    bLinearSpring=true
    gravityScale=0
    lODThreshold=1
    linearSpringConstant=20
    alphaBoolBlend {blendOption=Linear}
    boundBone {boneName=dyn_collar}
    physicsBodyDefinitions [{
      sphereCollisionRadius=0
      boundBone {boneName=dyn_collar}
      boxExtents=<30 30 30>
      constraintSetup {
       twistAxis=AxisZ
       angularLimitsMax=<3 7.5 0>
       angularLimitsMin=<-3 -7.5 0>
       angularTarget=<0 0 0>
       linearAxesMax=<1.5 6 5>
       linearAxesMin=<-1.5 -6 -0.5>
      }
      localJointOffset=<0 0 0>
     }]
   }
   n16 = "Constraint - dyn_feather_r_bottom_base (PP)"!(ComponentPose=@n15, Alpha=1, ConstraintWeights_0=0.6944, ConstraintWeights_1:float=0.3056) -> (Pose) {
    alphaBoolBlend {blendOption=Linear}
    boneToModify {boneName=dyn_feather_r_bottom_base}
    constraintSetup [{targetBone {boneName=pelvis}}, {targetBone {boneName=thigh_twist_01_r}}]
    constraintWeights [1,1]
   }
   n17 = "Constraint - dyn_feather_l_top_base (PP)"!(ComponentPose=@n16, Alpha=1, ConstraintWeights_0=0.8473, ConstraintWeights_1:float=0.1527) -> (Pose) {
    alphaBoolBlend {blendOption=Linear}
    boneToModify {boneName=dyn_feather_l_top_base}
    constraintSetup [{targetBone {boneName=pelvis}}, {targetBone {boneName=thigh_twist_01_l}}]
    constraintWeights [1,1]
   }
   n18 = "Constraint - dyn_feather_l_bottom_base (PPP)"!(ComponentPose=@n17, Alpha=1, ConstraintWeights_0=0.6513, ConstraintWeights_1:float=0.3476, ConstraintWeights_2:float=0.001) -> (Pose) {
    alphaBoolBlend {blendOption=Linear}
    boneToModify {boneName=dyn_feather_l_bottom_base}
    constraintSetup [{targetBone {boneName=pelvis}}, {targetBone {boneName=thigh_twist_01_l}}, {targetBone {boneName=thigh_l}}]
    constraintWeights [1,1,1]
   }
   n19 = "Constraint - dyn_feather_r_top_base (PP)"!(ComponentPose=@n18, Alpha=1, ConstraintWeights_0=0.7852, ConstraintWeights_1:float=0.2148) -> (Pose) {
    alphaBoolBlend {blendOption=Linear}
    boneToModify {boneName=dyn_feather_r_top_base}
    constraintSetup [{targetBone {boneName=pelvis}}, {targetBone {boneName=thigh_twist_01_r}}]
    constraintWeights [1,1]
   }
   n20 = Anim Dynamics - Bone: dyn_feather_r_bottom_parent(ComponentPose=@n19, Alpha=1, ExternalForce="0,0,0") {
    alpha=0.9
    angularDampingOverride=0.99
    angularSpringConstant=125
    bGravityOverrideInSimSpace=true
    bUseGravityOverride=true
    gravityScale=0.5
    linearDampingOverride=0.99
    numSolverIterationsPostUpdate=4
    numSolverIterationsPreUpdate=16
    alphaBoolBlend {blendOption=Linear}
    boundBone {boneName=dyn_feather_r_bottom_parent}
    physicsBodyDefinitions [{
      sphereCollisionRadius=0
      boundBone {boneName=dyn_feather_r_bottom_parent}
      boxExtents=<6 8 2>
      constraintSetup {angularTargetAxis=AxisZ, twistAxis=AxisZ, angularLimitsMax=<0 15 10>, angularLimitsMin=<0 -15 -10>, angularTarget=<0 0 1>}
      localJointOffset=<2.5 0 0>
     }]
   }
   n21 = Anim Dynamics - Bone: dyn_feather_l_top_parent(ComponentPose=@n20, Alpha=1, ExternalForce="0,0,0") {
    alpha=0.9
    angularDampingOverride=0.99
    angularSpringConstant=125
    bGravityOverrideInSimSpace=true
    bUseGravityOverride=true
    gravityScale=0.5
    linearDampingOverride=0.99
    numSolverIterationsPostUpdate=4
    numSolverIterationsPreUpdate=16
    alphaBoolBlend {blendOption=Linear}
    boundBone {boneName=dyn_feather_l_top_parent}
    physicsBodyDefinitions [{
      sphereCollisionRadius=0
      boundBone {boneName=dyn_feather_l_top_parent}
      boxExtents=<6 8 2>
      constraintSetup {angularTargetAxis=AxisZ, twistAxis=AxisZ, angularLimitsMax=<0 15 10>, angularLimitsMin=<0 -15 -10>, angularTarget=<0 0 1>}
      localJointOffset=<2 0 0>
     }]
   }
   n22 = Anim Dynamics - Bone: dyn_feather_l_bottom_parent(ComponentPose=@n21, Alpha=1, ExternalForce="0,0,0") {
    alpha=0.9
    angularDampingOverride=0.99
    angularSpringConstant=125
    bGravityOverrideInSimSpace=true
    bUseGravityOverride=true
    gravityScale=0.5
    linearDampingOverride=0.99
    numSolverIterationsPostUpdate=4
    numSolverIterationsPreUpdate=16
    alphaBoolBlend {blendOption=Linear}
    boundBone {boneName=dyn_feather_l_bottom_parent}
    physicsBodyDefinitions [{
      sphereCollisionRadius=0
      boundBone {boneName=dyn_feather_l_bottom_parent}
      boxExtents=<6 8 2>
      constraintSetup {angularTargetAxis=AxisZ, twistAxis=AxisZ, angularLimitsMax=<0 15 10>, angularLimitsMin=<0 -15 -10>, angularTarget=<0 0 1>}
      localJointOffset=<2 0 0>
     }]
   }
   n23 = Anim Dynamics - Bone: dyn_feather_r_top_parent(ComponentPose=@n22, Alpha=1, ExternalForce="0,0,0") {
    alpha=0.9
    angularDampingOverride=0.99
    angularSpringConstant=125
    bGravityOverrideInSimSpace=true
    bUseGravityOverride=true
    gravityScale=0.5
    linearDampingOverride=0.99
    numSolverIterationsPostUpdate=4
    numSolverIterationsPreUpdate=16
    alphaBoolBlend {blendOption=Linear}
    boundBone {boneName=dyn_feather_r_top_parent}
    physicsBodyDefinitions [{
      sphereCollisionRadius=0
      boundBone {boneName=dyn_feather_r_top_parent}
      boxExtents=<6 8 2>
      constraintSetup {angularTargetAxis=AxisZ, twistAxis=AxisZ, angularLimitsMax=<0 15 10>, angularLimitsMin=<0 -15 -10>, angularTarget=<0 0 1>}
      localJointOffset=<2 0 0>
     }]
   }
   n24 = "Transform (Modify) Bone - Bone: dyn_leather_r_base"(
    ComponentPose=@n23
    Alpha=Get isCrouch() -> (isCrouch:double)
    Translation="-4.608,0.000,-1.767"
    Rotation="10.285,-41.234,-11.116"
    Scale="1,1,1"
   ) {
    rotationMode=BMM_Additive
    translationMode=BMM_Additive
    alphaBoolBlend {blendOption=Linear}
    alphaScaleBiasClamp {bInterpResult=true, interpSpeedDecreasing=25, interpSpeedIncreasing=20}
    boneToModify {boneName=dyn_leather_r_base}
    rotation=<rot 10.285 -41.234 -11.116>
    translation {x=-4.608, z=-1.767}
   }
   n25 = RigidBody(ComponentPose=@n24) {
    bClampLinearTranslationLimitToRefPose=true
    bOverrideWorldGravity=true
    overridePhysicsAsset=&/F_MED_Raven_Body_02_Skirt_Physics
    simulationSpace=BaseBoneSpace
    alphaBoolBlend {blendOption=Linear}
    baseBoneRef {boneName=root}
    componentAppliedLinearAccClamp=<3000 3000 3000>
    componentLinearAccScale=<0.15 0.15 0.15>
    componentLinearVelScale=<0.15 0.15 0.15>
    overrideWorldGravity {z=-980}
   }
   Output Pose(Result=Component To Local(ComponentPose=@n25)) {layerGroup=None}
  ]}
 EventGraph {nodes [
   Event BlueprintUpdateAnimation()
   > n1 = Sequence!(execute) -> (then_0, then_1, then_2:Exec)
   n2 = Get PawnLastLocation() -> (PawnLastLocation:Vector)
   n3 = SelectFloat(
    A=Make Perlin Noise and Remap!(Value:float=BreakVector(InVec=@n2).Z, RangeOutMin:float=-25000, RangeOutMax:float=25000) -> (ReturnValue:float)
    B=0
    bPickA=Get bIsSkyDiving()
   )
   n4 = SelectFloat(A=float * float(A=@n3, B=0.5), B=@n3, bPickA=Get bIsParachuteOpened())
   > Set falling_anim_dyn_up_force_vec3(falling_anim_dyn_up_force_vec3:Vector=MakeVector(X=@n4, Y=@n4)) -> (Output_Get:Vector)
   Set isCrouch(execute=@n1.then_2, isCrouch:double="To Float (Boolean)"(InBool=Get bIsCrouching())) -> (Output_Get:double)
   n5 = Get bIsCrouchSprinting()
   > Set isCrouchMove(isCrouchMove:double="To Float (Boolean)"(InBool=@n5)) -> (Output_Get:double)
   > Branch(Condition=@n5)
   > Set isCrouch(isCrouch=0)
   n6 = Branch(execute=@n1.then_1, Condition=Get bIsSkyDiving())
   n7 = Get raven_clothing_sim_interactor() -> (raven_clothing_sim_interactor:Clothing Simulation Interactor Nv Object Reference)
   > SetAnimDriveSpringStiffness!(execute, self:Clothing Simulation Interactor Object Reference=@n7, InStiffness:float=20) -> (then)
   > SetAnimDriveDamperStiffness!(execute, self:Clothing Simulation Interactor Nv Object Reference=@n7, InStiffness:float=10) -> (then)
   > Do Once(Reset=@n8.then, Start Closed=true)
   > ForceClothNextUpdateTeleportAndReset(self:Skeletal Mesh Component Object Reference=Get MeshToCopy() -> (MeshToCopy:Skeletal Mesh Component Object Reference))
   SetAnimDriveSpringStiffness(execute=@n6.else, self=@n7, InStiffness=0)
   > n8 = SetAnimDriveDamperStiffness(self=@n7, InStiffness=0)
   Event BlueprintInitializeAnimation()
   > Set random_stream(random_stream:Random Stream Structure=MakeRandomStream!(InitialSeed:int=0) -> (ReturnValue:Random Stream Structure)) -> (Output_Get:Random Stream Structure)
   > n9 = Cast To ClothingSimulationInteractorNv!(
    execute
    Object=GetClothingSimulationInteractor!(self:Skeletal Mesh Component Object Reference=GetPartSkeletalMeshComponent(), ClothingAssetName:Name=None) -> (ReturnValue:Clothing Simulation Interactor Object Reference)
   ) -> (then, CastFailed, AsClothing Simulation Interactor Nv:Clothing Simulation Interactor Nv Object Reference) {is=title K2Node_DynamicCast}
   > Set raven_clothing_sim_interactor(raven_clothing_sim_interactor:Clothing Simulation Interactor Nv Object Reference=@n9) -> (Output_Get:Clothing Simulation Interactor Nv Object Reference)
   TryGetPawnOwner()
  ]}
]
variables [
 falling_anim_dyn_up_force_vec3 {type=Vector}
 isCrouch {type=number}
 isCrouchMove {type=number}
 MeshToCopy {type=SkeletalMeshComponent}
 random_stream {type=RandomStream, default {initialSeed=0, seed=0}}
 raven_clothing_sim_interactor {type=ClothingSimulationInteractorNv, default=null}
]
